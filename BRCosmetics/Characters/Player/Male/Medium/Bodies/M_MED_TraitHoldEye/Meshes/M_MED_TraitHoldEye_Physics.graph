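# Physics asset for the M_MED_TraitHoldEye male medium player body, covering the left and right dangling ammo boxes and the thighs.
# Each side has a small capsule on the ammo box root bone, constrained to a box-shaped sub-bone body with swing and twist limit angles set.
# Unconstrained capsules on both thighs supply leg volumes, likely as colliders for the ammo boxes.
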PhysicsAsset /BRCosmetics/Characters/Player/Male/Medium/Bodies/M_MED_TraitHoldEye/Meshes/M_MED_TraitHoldEye_Physics
bodies {
 dyn_ammo_box_l [{shapeType=Capsule, length=1, radius=1}]
 dyn_ammo_box_r [{shapeType=Capsule, length=1, radius=1}]
 dyn_ammo_box_sub_l [{shapeType=Box, extentX=5.1055, extentY=1.5504, extentZ=12.4175, center=<6.3692 0.2342 -0.227>, rotation=<rot 0 -88.5322 -87.0247>}]
 dyn_ammo_box_sub_r [{shapeType=Box, extentX=5.1055, extentY=1.5504, extentZ=12.4175, center=<6.3692 -0.2342 -0.227>, rotation=<rot 0 -91.4678 -87.0247>}]
 thigh_l [{shapeType=Capsule, length=23.5, radius=8, center=<-17.4529 0.0002 0>, rotation=<rot 87.3701 0 0>}]
 thigh_r [{shapeType=Capsule, length=23.5, radius=8, center=<17.4529 -0.0002 0>, rotation=<rot 87.3701 0 180>}]
}
constraints [
 {bone1Name=dyn_ammo_box_sub_l, bone2Name=dyn_ammo_box_l, swing1LimitDegrees=45, swing2LimitDegrees=45, twistLimitDegrees=30}
 {bone1Name=dyn_ammo_box_sub_r, bone2Name=dyn_ammo_box_r, swing1LimitDegrees=45, swing2LimitDegrees=45, twistLimitDegrees=30}
]
